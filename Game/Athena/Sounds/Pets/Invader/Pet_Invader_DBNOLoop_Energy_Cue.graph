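# Sound cue for the Invader pet's energy loop sound, named for the DBNO (down but not out) state.
# A single wave player feeds the root with the Pet_Invader_DBNOLoop_Energy wave, with no randomization, mixing or loop node in the cue.
# It plays at reduced volume through the creature pets sound class, using a very-close pet attenuation that keeps it audible only at short range.
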SoundCue /Game/Athena/Sounds/Pets/Invader/Pet_Invader_DBNOLoop_Energy_Cue
root=WavePlayer_0
nodes [{id=WavePlayer_0, properties {soundWaveAssetPtr=&/Pet_Invader_DBNOLoop_Energy}}]
properties {
 attenuationSettings=/Game/Sounds/Attenuation/Fort_Pets_VeryClose_Athena_Att
 maxDistance=360
 soundClassObject=&Fort_Creatures_Pets
 volumeMultiplier=0.5152
}
